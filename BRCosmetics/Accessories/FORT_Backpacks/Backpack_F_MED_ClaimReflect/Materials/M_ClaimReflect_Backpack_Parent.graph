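# Parent material for the ClaimReflect backpack that renders an animated miniature planet surface where a tech art mask channel selects it.
# The planet attributes blend between the shared character PreFX and PostFX functions.
# One packed noise texture, sampled with bump-offset parallax, drives water, foam, desert, land, mountains and normals.
# Panning distorted clouds cast shadows offset along the sky light direction, and Fresnel adds an atmosphere glow.
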Material /BRCosmetics/Accessories/FORT_Backpacks/Backpack_F_MED_ClaimReflect/Materials/M_ClaimReflect_Backpack_Parent
graph {
 properties {
  bUseMaterialAttributes=true
  bUsedWithInstancedStaticMeshes=true
  bUsedWithMeshDeformer=true
  bUsedWithMorphTargets=true
  bUsedWithSkeletalMesh=true
  pixelDepthOffsetMode=PDOM_Legacy
  shadingModel=MSM_FromMaterialExpression
  usedShadingModels=Default Lit | Subsurface | Clear Coat | Subsurface Profile | Cloth
 }
 nodes [
  n1 = NamedRerouteUsage()
  n2 = NamedRerouteUsage()
  n3 = Add(
   A=ConstantBiasScale(None=LinearInterpolate(A=@n1, Alpha=@n2) {constB=0}) {bias=-0.5, scale=2}
   B=ScalarParameter() {defaultValue=0.0713, group=Landscape, parameterName=WaterLevel}
  )
  n4 = Saturate(None=@n3)
  Ceil(None=@n4)
  n5 = MaterialFunctionCall() {materialFunction=&MF_CharacterParent_2023_PreFX}
  MaterialFunctionCall(
   Material=BlendMaterialAttributes(
    A=NamedRerouteDeclaration(None=@n5.Result) {name=InMaterial, nodeColor=<rgba 0 0.5781 1 1>}
    B=NamedRerouteUsage()
    Alpha=ChannelMaskParameter(None=@n5.Tech Art Mask) {group=Landscape, parameterName=PlanetFXMask}.R
   )
   M=@n5.M
   Pre Skinned Local Position=@n5.M
   Distance Blend=@n5.M
   Fresnel=@n5.M
  ) {materialFunction=&MF_CharacterParent_2023_PostFX}
  n6 = VectorParameter() {parameterName=Curvature Axis, defaultValue=<rgba 0 -1 0 0>}
  MaterialFunctionCall(X=@n6.R, Y=@n6.R, Z=@n6.R) {materialFunction=&MakeFloat3}
  n7 = VectorParameter() {group=Landscape, parameterName=LandscapeTextureOffset, defaultValue=<rgba -1.6738 0.2336 0 1>}
  n8 = Multiply(A=Add(A=TextureCoordinate(), B=ComponentMask(None=@n7.RGB) {g=true, r=true})) {constB=2.5}
  n9 = TextureSample(UVs=@n8) {
   samplerType=SAMPLERTYPE_LinearColor
   texture=&/T_ClaimReflectBackpack_Noise1
  }
  n10 = BumpOffset(Coordinate=@n8, Height=Add(A=ScalarParameter() {defaultValue=-1, group=Landscape, parameterName=LandscapeBaseHeight}, B=@n9.B)) {heightRatio=0.075}
  n11 = TextureSample(UVs=@n10) {
   samplerSource=SSM_Wrap_WorldGroupSettings
   samplerType=SAMPLERTYPE_LinearColor
   texture=&T_ClaimReflectBackpack_Noise1
  }
  NamedRerouteDeclaration(None=@n11.B) {name=LandscapeHeight, nodeColor=<rgba 0.7813 0 1 1>}
  NamedRerouteDeclaration(None=SphereMask(A=@n3, B=Constant(), Radius=ScalarParameter() {group=Landscape, parameterName=WaterEdgeRadius}) {attenuationRadius=0.15, hardnessPercent=0}) {name=WaterFoamMask, nodeColor=<rgba 0.5937 1 0 1>}
  NamedRerouteDeclaration(None=GetMaterialAttributes(None=NamedRerouteUsage()) {attributeGetTypes [#1]}.Normal) {name=InNormal, nodeColor=<rgba 1 0.0234 0 1>}
  n12 = TextureSample(UVs=Panner(Coordinate=Multiply(A=@n10) {constB=15}) {speedX=-0.1}) {samplerSource=SSM_Wrap_WorldGroupSettings, samplerType=SAMPLERTYPE_Normal, texture=/Game/Global/Textures/Default/Tileable_Patterns/1Normal}
  NamedRerouteDeclaration(None=@n12.RGB) {name=OutWaterNormal, nodeColor=<rgba 0 0.2969 1 1>}
  n13 = ScalarParameter() {defaultValue=0.8, group=Landscape, parameterName=DesertLandNormalIntensity}
  n14 = ScalarParameter() {defaultValue=-0.5, group=Landscape, parameterName=LandNormalIntensity}
  n15 = NamedRerouteUsage()
  n16 = LinearInterpolate(
   A=MaterialFunctionCall(Normal=NamedRerouteUsage(), Flatness=LinearInterpolate(A=@n13, B=@n14, Alpha=@n15)) {materialFunction=&FlattenNormal}.Result
   B=MaterialFunctionCall(Normal=NamedRerouteUsage(), Flatness=ScalarParameter() {defaultValue=0.9, group=Landscape, parameterName=WaterNormalIntensity}) {materialFunction=&FlattenNormal}.Result
   Alpha=NamedRerouteUsage()
  )
  n17 = MaterialFunctionCall(Normal=NamedRerouteUsage(), Flatness=ScalarParameter() {group=Landscape, parameterName=CloudNormalIntensity}) {materialFunction=&FlattenNormal}
  n18 = NamedRerouteUsage()
  NamedRerouteDeclaration(None=MaterialFunctionCall(BaseNormal=NamedRerouteUsage(), AdditionalNormal=LinearInterpolate(A=@n16, B=@n17.Result, Alpha=@n18)) {materialFunction=&BlendAngleCorrectedNormals}.Result) {name=OutNormal, nodeColor=<rgba 0.6172 1 0 1>}
  n19 = Fresnel(ExponentIn=ScalarParameter() {defaultValue=3, group=Landscape, parameterName=AtmosphereFresnelExponent}, Normal=VertexNormalWS()) {baseReflectFraction=0}
  n20 = VectorParameter() {group=Landscape, parameterName=AtmosphereEmissiveColor, defaultValue=<rgba 0.3959 0.5767 0.75 0.2>}
  NamedRerouteDeclaration(None=Multiply(A=Multiply(A=@n19, B=@n20.RGB), B=@n20.A)) {name=OutEmissive, nodeColor=<rgba 0.0781 1 0 1>}
  n21 = Multiply(A=NamedRerouteUsage(), B=ScalarParameter() {defaultValue=0.5, group=Landscape, parameterName=WaterMetalness})
  n22 = NamedRerouteUsage()
  NamedRerouteDeclaration(None=LinearInterpolate(A=LinearInterpolate(A=@n21, Alpha=@n22), Alpha=NamedRerouteUsage()) {constB=0}) {name=OutMetallic, nodeColor=<rgba 0.7813 0 1 1>}
  n23 = BumpOffset(
   Coordinate=Multiply(A=TextureCoordinate()) {constB=2.5}
   Height=ScalarParameter() {defaultValue=1, group=Landscape, parameterName=CloudsHeight}
  ) {heightRatio=0.1}
  n24 = Multiply(A=@n23, B=Constant2Vector() {g=1, r=0.5})
  n25 = TextureSample(UVs=Add(A=@n23) {constB=0.324}) {
   samplerSource=SSM_Wrap_WorldGroupSettings
   samplerType=SAMPLERTYPE_LinearColor
   texture=&T_ClaimReflectBackpack_Noise1
  }
  n26 = Add(A=Panner(Coordinate=Multiply(A=@n24) {constB=0.75}) {speedX=-0.02}, B=Multiply(A=Subtract(A=@n25.B) {constB=0.5}) {constB=0.2})
  n27 = TextureSample(UVs=@n26) {
   samplerSource=SSM_Wrap_WorldGroupSettings
   samplerType=SAMPLERTYPE_LinearColor
   texture=&T_ClaimReflectBackpack_Noise1
  }
  n28 = ScalarParameter() {defaultValue=9, group=Landscape, parameterName=CloudExponent}
  n29 = ScalarParameter() {defaultValue=50, group=Landscape, parameterName=CloudIntensity}
  n30 = Multiply(A=Power(Base=NamedRerouteDeclaration(None=@n27.B) {name=CloudsNoise, nodeColor=<rgba 1 0 0.6328 1>}, Exp=@n28), B=@n29)
  NamedRerouteDeclaration(None=LinearInterpolate(A=Saturate(None=@n30), Alpha=NamedRerouteUsage()) {constB=0}) {name=CloudsMask, nodeColor=<rgba 1 0 0.8203 1>}
  n31 = ConstantBiasScale(None=AppendVector(A=@n27.R, B=@n27.R)) {bias=-0.5, scale=2}
  NamedRerouteDeclaration(None=DeriveNormalZ(InXY=Multiply(A=@n31) {constB=-1})) {name=CloudNormal, nodeColor=<rgba 0.0547 1 0 1>}
  NamedRerouteDeclaration(None=SphereMask(A=@n4, B=Constant() {r=1}, Radius=ScalarParameter() {defaultValue=0.8, group=Landscape, parameterName=MountainTopRadius}) {attenuationRadius=0.8, hardnessPercent=30}) {name=TopMask, nodeColor=<rgba 1 0.3047 0 1>}
  n32 = LinearInterpolate(
   A=LinearInterpolate(Alpha=NamedRerouteUsage()) {constA=0.5, constB=0.5}
   B=ScalarParameter() {defaultValue=0.9, group=Landscape, parameterName=WaterSpecular}
   Alpha=NamedRerouteUsage()
  ) {constA=1, constB=0.5}
  n33 = NamedRerouteUsage()
  NamedRerouteDeclaration(None=LinearInterpolate(
    A=LinearInterpolate(A=@n32, Alpha=@n33) {constB=0.5}
    B=ScalarParameter() {defaultValue=0.1, group=Landscape, parameterName=CloudSpecular}
    Alpha=NamedRerouteUsage()
   )) {name=OutSpecular, nodeColor=<rgba 0 0.3437 1 1>}
  NamedRerouteDeclaration(None=OneMinus(None=SmoothStep(Value=@n4) {constMax=0.1})) {name=WaterMask, nodeColor=<rgba 0 1 0.4609 1>}
  n34 = SmoothStep(Value=NamedRerouteUsage()) {constMax=0.5, constMin=0.2}
  NamedRerouteDeclaration(None=@n34) {name=DesertArea, nodeColor=<rgba 0 1 0.7188 1>}
  n35 = Transform(None=Multiply(A=SkyAtmosphereLightDirection(), B=ScalarParameter() {defaultValue=0.05, group=Landscape, parameterName=CloudShadowDistance}) {constB=0.1}) {transformSourceType=TRANSFORMSOURCE_World, transformType=TRANSFORM_Tangent}
  n36 = AppendVector(A=@n26, B=Constant())
  n37 = TextureSample(UVs=ComponentMask(None=Add(A=@n35, B=@n36)) {g=true, r=true}) {
   samplerSource=SSM_Wrap_WorldGroupSettings
   samplerType=SAMPLERTYPE_LinearColor
   texture=&T_ClaimReflectBackpack_Noise1
  }
  n38 = Saturate(None=Multiply(A=Power(Base=@n37.B, Exp=@n28), B=@n29))
  NamedRerouteDeclaration(None=LinearInterpolate(A=SmoothStep(Value=@n38), Alpha=NamedRerouteUsage()) {constB=0}) {name=CloudShadowMask, nodeColor=<rgba 0.2891 1 0 1>}
  n39 = TextureSample(UVs=BumpOffset(Coordinate=TextureCoordinate(), Height=ScalarParameter() {defaultValue=0.09, group=Landscape, parameterName=RingEdgeHeight})) {
   samplerSource=SSM_Wrap_WorldGroupSettings
   samplerType=SAMPLERTYPE_LinearColor
   texture=&T_ClaimReflectBackpack_Noise1
  }
  NamedRerouteDeclaration(None=Step(Y=ScalarParameter() {defaultValue=0.95, group=Landscape, parameterName=RingEdgeStep}, X=OneMinus(None=@n39.A))) {name=Edge, nodeColor=<rgba 0.6641 0 1 1>}
  n40 = NamedRerouteUsage()
  n41 = Constant()
  n42 = ScalarParameter() {defaultValue=1.464, group=Landscape, parameterName=EdgeMaskingRadius}
  n43 = SmoothStep(Value=OneMinus(None=SphereMask(A=@n4, B=@n41, Radius=@n42) {attenuationRadius=0.2, hardnessPercent=0})) {constMax=0.2}
  NamedRerouteDeclaration(None=Saturate(None=Subtract(A=@n40, B=@n43))) {name=EdgeMask, nodeColor=<rgba 1 0.7734 0 1>}
  NamedRerouteDeclaration(None=ChannelMaskParameter(None=@n5.Tech Art Mask) {group=Landscape, maskChannel=Green, parameterName=SeamFixFXMask, defaultValue=<rgba 0 1 0 0>}.R) {name=SeamFixMask, nodeColor=<rgba 0.0547 1 0 1>}
  n44 = AppendVector(A=@n11.R, B=@n11.R)
  NamedRerouteDeclaration(None=DeriveNormalZ(InXY=ConstantBiasScale(None=@n44) {bias=-0.5, scale=2})) {name=LandscapeNormals, nodeColor=<rgba 0.2422 1 0 1>}
  n45 = SetMaterialAttributes!(
   MaterialAttributes=NamedRerouteUsage()
   Base Color=NamedRerouteUsage()
   Metallic=NamedRerouteUsage()
   Specular=NamedRerouteUsage()
   Roughness=NamedRerouteUsage()
   Normal=NamedRerouteUsage()
   Emissive Color=NamedRerouteUsage()
   Pixel Depth Offset
  ) {attributeSetTypes [
    #2
    #3
    #4
    #5
    #1
    #6
    #7
   ]}
  NamedRerouteDeclaration(None=@n45) {name=OutMaterial, nodeColor=<rgba 0.5 1 0 1>}
  n46 = TextureSample(UVs=Add(A=Multiply(A=@n10) {constB=0.5}, B=ScalarParameter() {defaultValue=0.25, group=Landscape, parameterName=LowFreqLandscapeOffset}) {constB=0.15}) {
   samplerSource=SSM_Wrap_WorldGroupSettings
   samplerType=SAMPLERTYPE_LinearColor
   texture=&T_ClaimReflectBackpack_Noise1
  }
  NamedRerouteDeclaration(None=@n46.B) {name=LowFreqLandscapeHeight, nodeColor=<rgba 0 0.6016 1 1>}
  n47 = NamedRerouteUsage()
  n48 = LinearInterpolate(
   A=VectorParameter() {group=Landscape, parameterName=LandColor, defaultValue=<rgba 0.5156 0.2529 0.1315 1>}.RGB
   B=VectorParameter() {group=Landscape, parameterName=LandColor2, defaultValue=<rgba 0.0165 0.1094 0.0321 1>}.RGB
   Alpha=Multiply(A=Power(Base=@n47) {constExponent=0.075}, B=@n34)
  )
  n49 = VectorParameter() {group=Landscape, parameterName=MountainColor, defaultValue=<rgba 0.0777 0.0781 0.0477 1>}
  n50 = Multiply(A=NamedRerouteUsage(), B=Floor(None=@n34))
  n51 = VectorParameter() {group=Landscape, parameterName=WaterColor, defaultValue=<rgba 0.0535 0.0672 0.1771 1>}
  n52 = Multiply(A=@n51.RGB, B=ScalarParameter() {defaultValue=1.5, group=Landscape, parameterName=WaterEdgeIntensity}) {constB=2}
  n53 = NamedRerouteUsage()
  n54 = LinearInterpolate(
   A=LinearInterpolate(A=LinearInterpolate(A=@n48, B=@n49.RGB, Alpha=@n50), B=LinearInterpolate(A=@n51.RGB, B=@n52, Alpha=@n53), Alpha=NamedRerouteUsage())
   B=VectorParameter() {group=Landscape, parameterName=EdgeColor, defaultValue=<rgba 0.2051 0.1912 0.1946 1>}.RGB
   Alpha=NamedRerouteUsage()
  )
  n55 = LinearInterpolate(
   A=@n54
   B=Multiply(A=@n54, B=ScalarParameter() {defaultValue=0.1, group=Landscape, parameterName=CloudShadowIntensity}) {constB=0.6}
   Alpha=NamedRerouteUsage()
  )
  n56 = VectorParameter() {group=Landscape, parameterName=CloudColor, channelNames {a=CloudOpacity, b="", g="", r=""}, defaultValue=<rgba 0.5625 0.7069 1 1>}
  n57 = NamedRerouteUsage()
  n58 = ScalarParameter() {defaultValue=2, group=Landscape, parameterName=CloudShadingExponent}
  n59 = Multiply(A=Multiply(A=@n56, B=Power(Base=@n57, Exp=@n58)), B=@n56.CloudOpacity) {constB=0.9}
  n60 = NamedRerouteUsage()
  NamedRerouteDeclaration(None=Add(
    A=LinearInterpolate(A=@n55, B=@n59, Alpha=@n60)
    B=VectorParameter() {group=Landscape, parameterName=OverallTint, defaultValue=<rgba 0.0331 0.096 0.1354 1>}.RGB
   )) {name=OutColor, nodeColor=<rgba 1 0 0.4219 1>}
  NamedRerouteDeclaration(None=@n4) {name=LandGradient, nodeColor=<rgba 0 0.7422 1 1>}
  n61 = NamedRerouteUsage()
  n62 = NamedRerouteUsage()
  n63 = OneMinus(None=NamedRerouteUsage())
  n64 = LinearInterpolate(
   A=LinearInterpolate(A=LinearInterpolate(Alpha=@n61) {constA=0.8, constB=0.9}, Alpha=Max(A=@n62, B=@n63))
   B=ScalarParameter() {defaultValue=0.3, group=Landscape, parameterName=WaterRoughness}
   Alpha=NamedRerouteUsage()
  ) {constA=0.3, constB=0.3}
  n65 = NamedRerouteUsage()
  NamedRerouteDeclaration(None=LinearInterpolate(A=LinearInterpolate(A=@n64, Alpha=@n65) {constB=0.3}, Alpha=NamedRerouteUsage())) {name=OutRoughness, nodeColor=<rgba 1 0.9375 0 1>}
  n66 = TextureObject() {
   samplerType=SAMPLERTYPE_Masks
   texture=&T_ClaimReflectBackpack_Noise1
  }
  n67 = ScalarParameter() {defaultValue=8, parameterName=Param_2}
  n68 = TextureCoordinate()
  n69 = TextureSample(UVs=MaterialFunctionCall!(
    Heightmap Texture=@n66
    Height Ratio=ScalarParameter() {defaultValue=0.08, parameterName="Height Ratio (S)"}
    Min Steps
    Max Steps=@n67
    UVs=Multiply(A=@n68)
    Heightmap Channel=Constant4Vector() {constant=<rgba 0 0 1 0>}.RGBA
    Reference Plane=ScalarParameter() {defaultValue=1, parameterName=Param_1}
    -----------------------------
    Use World Coordinates
    Specify Manual Texture Size
    Manual Texture Size
    ------------------------------
    "Render Shadows (Occlusion Mapping)"
    Light Vector=SkyAtmosphereLightDirection()
    Shadow Steps=@n67
    Shadow Penumbra
    Transform To VertexNormal
   ) {materialFunction=/Engine/Functions/Engine_MaterialFunctions01/Texturing/ParallaxOcclusionMapping}.Parallax UVs) {
   samplerType=SAMPLERTYPE_Masks
   texture=&T_ClaimReflectBackpack_Noise1
  }
  Reroute(None=@n69.B)
  Subtract(A=@n69.R, B=Fresnel() {baseReflectFraction=0})
 ]
}
