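# Animation Blueprint for the ClaimReflect backpack cosmetic that procedurally floats fifteen pieces, spins the base and adjusts the pack to spine bend.
# Each update computes phase-offset sine waves from real time and a wrapped spin angle from RotationSpeed.
# It also maps spine_05's yaw delta from reference pose, relative to spine_01, to clamped pack rotation and translation.
# The AnimGraph applies these values through chained additive Transform (Modify) Bone nodes.
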AnimBlueprint /BRCosmetics/Accessories/FORT_Backpacks/Backpack_F_MED_ClaimReflect/Meshes/F_MED_ClaimReflect_Pack_AnimBP
target_skeleton=&/F_MED_ClaimReflect_Pack_Skeleton
graphs [
 AnimGraph {nodes [
   n1 = "Transform (Modify) Bone - Bone: dyn_ClaimReflect_piece_01"!(
    ComponentPose=Local To Component(LocalPose=Input Pose())
    Alpha=1
    Translation_X:double=0
    Translation_Y:double=Get Piece01() -> (Piece01:double)
    Translation_Z:double=0
    Rotation="0,0,0"
    Scale="1,1,1"
   ) -> (Pose) {
    translationMode=BMM_Additive
    translationSpace=BCS_BoneSpace
    alphaBoolBlend {blendOption=Linear}
    boneToModify {boneName=dyn_ClaimReflect_piece_01}
   }
   n2 = Get Piece02() -> (Piece02:double)
   n3 = "Transform (Modify) Bone - Bone: dyn_ClaimReflect_piece_02"!(ComponentPose=@n1, Alpha=1, Translation_X:double=0, Translation_Y:double=@n2, Translation_Z:double=@n2, Rotation="0,0,0", Scale="1,1,1") -> (Pose) {
    rotationMode=BMM_Additive
    rotationSpace=BCS_BoneSpace
    translationMode=BMM_Additive
    translationSpace=BCS_BoneSpace
    alphaBoolBlend {blendOption=Linear}
    boneToModify {boneName=dyn_ClaimReflect_piece_02}
   }
   n4 = Get Piece03() -> (Piece03:double)
   n5 = "Transform (Modify) Bone - Bone: dyn_ClaimReflect_piece_03"!(ComponentPose=@n3, Alpha=1, Translation_X:double=0, Translation_Y:double=@n4, Translation_Z:double=@n4, Rotation="0,0,0", Scale="1,1,1") -> (Pose) {
    translationMode=BMM_Additive
    translationSpace=BCS_BoneSpace
    alphaBoolBlend {blendOption=Linear}
    boneToModify {boneName=dyn_ClaimReflect_piece_03}
   }
   n6 = "Transform (Modify) Bone - Bone: dyn_ClaimReflect_piece_04"!(
    ComponentPose=@n5
    Alpha=1
    Translation_X:double=0
    Translation_Y:double=Get Piece04() -> (Piece04:double)
    Translation_Z:double=0
    Rotation="0,0,0"
    Scale="1,1,1"
   ) -> (Pose) {
    translationMode=BMM_Additive
    translationSpace=BCS_BoneSpace
    alphaBoolBlend {blendOption=Linear}
    boneToModify {boneName=dyn_ClaimReflect_piece_04}
   }
   n7 = "Transform (Modify) Bone - Bone: dyn_ClaimReflect_piece_05"!(
    ComponentPose=@n6
    Alpha=1
    Translation_X:double=0
    Translation_Y:double=Get Piece05() -> (Piece05:double)
    Translation_Z:double=0
    Rotation="0,0,0"
    Scale="1,1,1"
   ) -> (Pose) {
    translationMode=BMM_Additive
    translationSpace=BCS_BoneSpace
    alphaBoolBlend {blendOption=Linear}
    boneToModify {boneName=dyn_ClaimReflect_piece_05}
   }
   n8 = "Transform (Modify) Bone - Bone: dyn_ClaimReflect_piece_06"!(
    ComponentPose=@n7
    Alpha=1
    Translation_X:double=0
    Translation_Y:double=Get Piece06() -> (Piece06:double)
    Translation_Z:double=0
    Rotation="0,0,0"
    Scale="1,1,1"
   ) -> (Pose) {
    translationMode=BMM_Additive
    translationSpace=BCS_BoneSpace
    alphaBoolBlend {blendOption=Linear}
    boneToModify {boneName=dyn_ClaimReflect_piece_06}
   }
   n9 = "Transform (Modify) Bone - Bone: dyn_ClaimReflect_piece_07"!(
    ComponentPose=@n8
    Alpha=1
    Translation_X:double=0
    Translation_Y:double=0
    Translation_Z:double=Get Piece07() -> (Piece07:double)
    Rotation="0,0,0"
    Scale="1,1,1"
   ) -> (Pose) {
    translationMode=BMM_Additive
    translationSpace=BCS_BoneSpace
    alphaBoolBlend {blendOption=Linear}
    boneToModify {boneName=dyn_ClaimReflect_piece_07}
   }
   n10 = Get Piece08() -> (Piece08:double)
   n11 = "Transform (Modify) Bone - Bone: dyn_ClaimReflect_piece_08"!(ComponentPose=@n9, Alpha=1, Translation_X:double=@n10, Translation_Y:double=@n10, Translation_Z:double=@n10, Rotation="0,0,0", Scale="1,1,1") -> (Pose) {
    translationMode=BMM_Additive
    translationSpace=BCS_BoneSpace
    alphaBoolBlend {blendOption=Linear}
    boneToModify {boneName=dyn_ClaimReflect_piece_08}
   }
   n12 = Get Piece09() -> (Piece09:double)
   n13 = "Transform (Modify) Bone - Bone: dyn_ClaimReflect_piece_09"!(
    ComponentPose=@n11
    Alpha=1
    Translation_X:double=@n12
    Translation_Y:double=@n12
    Translation_Z:double=@n12
    Rotation="0,0,0"
    Scale="1,1,1"
   ) -> (Pose) {
    translationMode=BMM_Additive
    translationSpace=BCS_BoneSpace
    alphaBoolBlend {blendOption=Linear}
    boneToModify {boneName=dyn_ClaimReflect_piece_09}
   }
   n14 = Get Piece10() -> (Piece10:double)
   n15 = "Transform (Modify) Bone - Bone: dyn_ClaimReflect_piece_10"!(ComponentPose=@n13, Alpha=1, Translation_X:double=@n14, Translation_Y:double=@n14, Translation_Z:double=0, Rotation="0,0,0", Scale="1,1,1") -> (Pose) {
    translationMode=BMM_Additive
    translationSpace=BCS_BoneSpace
    alphaBoolBlend {blendOption=Linear}
    boneToModify {boneName=dyn_ClaimReflect_piece_10}
   }
   n16 = Get Piece11() -> (Piece11:double)
   n17 = "Transform (Modify) Bone - Bone: dyn_ClaimReflect_piece_11"!(ComponentPose=@n15, Alpha=1, Translation_X:double=@n16, Translation_Y:double=@n16, Translation_Z:double=0, Rotation="0,0,0", Scale="1,1,1") -> (Pose) {
    translationMode=BMM_Additive
    translationSpace=BCS_BoneSpace
    alphaBoolBlend {blendOption=Linear}
    boneToModify {boneName=dyn_ClaimReflect_piece_11}
   }
   n18 = Get Piece12() -> (Piece12:double)
   n19 = "Transform (Modify) Bone - Bone: dyn_ClaimReflect_piece_12"!(ComponentPose=@n17, Alpha=1, Translation_X:double=@n18, Translation_Y:double=@n18, Translation_Z:double=0, Rotation="0,0,0", Scale="1,1,1") -> (Pose) {
    translationMode=BMM_Additive
    translationSpace=BCS_BoneSpace
    alphaBoolBlend {blendOption=Linear}
    boneToModify {boneName=dyn_ClaimReflect_piece_12}
   }
   n20 = Get Piece13() -> (Piece13:double)
   n21 = "Transform (Modify) Bone - Bone: dyn_ClaimReflect_piece_13"!(
    ComponentPose=@n19
    Alpha=1
    Translation_X:double=@n20
    Translation_Y:double=@n20
    Translation_Z:double=@n20
    Rotation="0,0,0"
    Scale="1,1,1"
   ) -> (Pose) {
    translationMode=BMM_Additive
    translationSpace=BCS_BoneSpace
    alphaBoolBlend {blendOption=Linear}
    boneToModify {boneName=dyn_ClaimReflect_piece_13}
   }
   n22 = Get Piece14() -> (Piece14:double)
   n23 = "Transform (Modify) Bone - Bone: dyn_ClaimReflect_piece_14"!(ComponentPose=@n21, Alpha=1, Translation_X:double=0, Translation_Y:double=@n22, Translation_Z:double=@n22, Rotation="0,0,0", Scale="1,1,1") -> (Pose) {
    translationMode=BMM_Additive
    translationSpace=BCS_BoneSpace
    alphaBoolBlend {blendOption=Linear}
    boneToModify {boneName=dyn_ClaimReflect_piece_14}
   }
   n24 = Get Piece15() -> (Piece15:double)
   n25 = "Transform (Modify) Bone - Bone: dyn_ClaimReflect_piece_15"!(ComponentPose=@n23, Alpha=1, Translation_X:double=@n24, Translation_Y:double=0, Translation_Z:double=@n24, Rotation="0,0,0", Scale="1,1,1") -> (Pose) {
    translationMode=BMM_Additive
    translationSpace=BCS_BoneSpace
    alphaBoolBlend {blendOption=Linear}
    boneToModify {boneName=dyn_ClaimReflect_piece_15}
   }
   n26 = "Transform (Modify) Bone - Bone: dyn_ClaimReflect_base"!(
    ComponentPose=@n25
    Alpha=1
    Translation="0,0,0"
    Rotation_Roll:float=Get CurrentRotation() -> (CurrentRotation:double)
    Rotation_Pitch:float=0
    Rotation_Yaw:float=0
    Scale="1,1,1"
   ) -> (Pose) {rotationMode=BMM_Additive, rotationSpace=BCS_BoneSpace, alphaBoolBlend {blendOption=Linear}, boneToModify {boneName=dyn_ClaimReflect_base}}
   n27 = "Transform (Modify) Bone - Bone: pack"(ComponentPose=@n26, Translation=Get backpack_translation(), Rotation=Get backpack_rotate()) {rotationMode=BMM_Additive, translationMode=BMM_Additive, alphaBoolBlend {blendOption=Linear}, boneToModify {boneName=pack}}
   Output Pose(Result=Component To Local(ComponentPose=@n27)) {layerGroup=None}
  ]}
 EventGraph {nodes [
   Event BlueprintUpdateAnimation()
   n1 = GetRealTimeSeconds()
   n2 = float + float(A=float * float(A=@n1, B=1) {is=title K2Node_PromotableOperator}, B=73)
   > Set Piece01(Piece01:double=float * float(A="Sin (Radians)"(A=@n2), B=0.6)) -> (Output_Get:double)
   n3 = float + float(A=float * float(A=@n1, B=1), B=20)
   > Set Piece02(Piece02:double=float * float(A="Sin (Radians)"(A=@n3), B=0.4)) -> (Output_Get:double)
   n4 = float + float(A=float * float(A=@n1, B=1), B=50)
   > Set Piece03(Piece03:double=float * float(A="Sin (Radians)"(A=@n4), B=0.7)) -> (Output_Get:double)
   n5 = float + float(A=float * float(A=@n1, B=1), B=40)
   > Set Piece04(Piece04:double=float * float(A="Sin (Radians)"(A=@n5), B=0.5)) -> (Output_Get:double)
   n6 = float + float(A=float * float(A=@n1, B=1), B=45)
   > Set Piece05(Piece05:double=float * float(A="Sin (Radians)"(A=@n6), B=0.4)) -> (Output_Get:double)
   n7 = float + float(A=float * float(A=@n1, B=1), B=30)
   > Set Piece06(Piece06:double=float * float(A="Sin (Radians)"(A=@n7), B=0.5)) -> (Output_Get:double)
   n8 = float + float(A=float * float(A=@n1, B=1), B=40)
   > Set Piece07(Piece07:double=float * float(A="Sin (Radians)"(A=@n8), B=0.3)) -> (Output_Get:double)
   n9 = float + float(A=float * float(A=@n1, B=1), B=70)
   > Set Piece08(Piece08:double=float * float(A="Sin (Radians)"(A=@n9), B=0.45)) -> (Output_Get:double)
   n10 = float + float(A=float * float(A=@n1, B=1), B=60)
   > Set Piece09(Piece09:double=float * float(A="Sin (Radians)"(A=@n10), B=0.5)) -> (Output_Get:double)
   n11 = float + float(A=float * float(A=@n1, B=1), B=80)
   > Set Piece10(Piece10:double=float * float(A="Sin (Radians)"(A=@n11), B=0.4)) -> (Output_Get:double)
   n12 = float + float(A=float * float(A=@n1, B=1), B=20)
   > Set Piece11(Piece11:double=float * float(A="Sin (Radians)"(A=@n12), B=0.3)) -> (Output_Get:double)
   n13 = float + float(A=float * float(A=@n1, B=1), B=10)
   > Set Piece12(Piece12:double=float * float(A="Sin (Radians)"(A=@n13), B=0.5)) -> (Output_Get:double)
   n14 = float + float(A=float * float(A=@n1, B=1), B=1)
   > Set Piece13(Piece13:double=float * float(A="Sin (Radians)"(A=@n14), B=0.35)) -> (Output_Get:double)
   n15 = float + float(A=float * float(A=@n1, B=0.5), B=10)
   > Set Piece14(Piece14:double=float * float(A="Sin (Radians)"(A=@n15), B=1)) -> (Output_Get:double)
   n16 = float + float(A=float * float(A=@n1, B=1), B=30)
   > Set Piece15(Piece15:double=float * float(A="Sin (Radians)"(A=@n16), B=0.5)) -> (Output_Get:double)
   n17 = Get RotationSpeed() -> (RotationSpeed:double)
   > Set CurrentRotation(CurrentRotation:double="Division (Whole and Remainder)"!(Dividend:double=float * float(A=@n1, B=@n17), Divisor:double=360) -> (Remainder:double, ReturnValue:int=0)) -> (Output_Get:double)
   n18 = GetOwnerSkeletalMeshComponent()
   n19 = BreakRotator(InRot=BreakTransform(InTransform=GetDeltaTransformFromRefPose(self=@n18, BoneName=spine_05, BaseName=spine_01)).Rotation)
   n20 = MapRangeClamped(Value=@n19.Yaw, InRangeA=-10, InRangeB=-50, OutRangeB=50)
   > Set backpack_rotate(backpack_rotate=MakeRotator(Roll=@n20, Pitch=float / float(A=@n20, B=5)))
   > Set backpack_translation(backpack_translation=MakeVector(Y=MapRangeClamped(Value=@n19.Yaw, InRangeA=-15, InRangeB=-50, OutRangeB=-10)))
   TryGetPawnOwner()
  ]}
]
variables [
 backpack_rotate {type=Rotator}
 backpack_translation {type=Vector}
 CurrentRotation {type=number, default=-351.9261}
 Piece01 {type=number, default=0.3903}
 Piece02 {type=number, default=-0.1187}
 Piece03 {type=number, default=-0.6925}
 Piece04 {type=number, default=0.3754}
 Piece05 {type=number, default=-0.1682}
 Piece06 {type=number, default=-0.1353}
 Piece07 {type=number, default=0.2252}
 Piece08 {type=number, default=-0.2416}
 Piece09 {type=number, default=0.4547}
 Piece10 {type=number, default=-0.0034}
 Piece11 {type=number, default=-0.089}
 Piece12 {type=number, default=0.3842}
 Piece13 {type=number, default=-0.1528}
 Piece14 {type=number, default=0.15}
 Piece15 {type=number, default=-0.1353}
 RotationSpeed {type=number, default=-12}
]
